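# Niagara crowd system for the buffet demo that burst-spawns a static crowd of mesh characters in a flattened torus, oriented by the negated torus vector with slight random rotation.
# Each particle gets random mesh indices and an animation offset for two mesh renderers.
# NMS_CrowdDynamicParameters receives clip indices and lengths, a male flag from comparing MeshIndex to 8, and a 25-second looping time from emitter age, apparently for WPO-driven animation.
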NiagaraSystem /Game/ArtTools/characterWPO/buffetDemo/NS_buffet_hugCrowd
dependencies {usedDataInterfaces [], usedDynamicInputs [], usedModules [&SystemState]}
emitters {crowdTest_v2 {
  emitterUpdateScript [
   &EmitterState(Life Cycle Mode=::NewEnumerator1, Loop Behavior=::NewEnumerator1, Loop Duration=1, UseLoopDelay=false)
   &SpawnBurst_Instantaneous(Spawn Count=550, Use Spawn Probability=false)
  ]
  particleSpawnScript [
   &TorusLocation(
    Align Torus With Axis=false
    Handle Radius=3600
    Large Radius=4000
    MaskSpawnGroup=false
    Module.Override Randomness=false
    Non Uniform Scale:Vector3f=<1 1 0>
    Offset=<0 150 0>
   )
   /SetVariables(
    Particles.MeshIndex=&RandomRangeInt(Minimum=0, Maximum=17)
    Particles.MeshIndex2:NiagaraInt32=&RandomRangeInt(Minimum=0, Maximum=17)
    Particles.animOffset:NiagaraFloat=&RandomRangeFloat(Minimum=0, Maximum=5, Random Seed=123)
   ) {isSetParameters=true}
   &InitialMeshOrientation(
    Orientation Axis=<0 -1 0>
    Orientation Coordinate Space=::World
    Orientation Vector=/Niagara/DynamicInputs/Vectors/NegateVector(Vector:Vector3f=$Output.TorusLocation.TorusVector)
    Rotation=&RandomRangeVector(Minimum=<0 0 -3>, Maximum=<0 0 3>)
    Use Orientation Vector=true
    Use Rotation Vector=true
   )
  ]
  particleUpdateScript [&/NMS_CrowdDynamicParameters(
    A_clipIndex1=1
    A_clipIndex2=2
    A_clipLength0=13.33
    A_clipLength1=7.222
    B_clipIndex0=3
    B_clipIndex1=4
    B_clipIndex2=5
    B_clipLength0=13.33
    B_clipLength1=7.222
    crowdTime:NiagaraFloat=&Multiply_Float(A=25, B=&Frac_Float(Float=&Divide_Float(A=$Emitter.Age, B=25)))
    male:NiagaraBool=&SetBoolByIntComparison(A=$StackContext.MeshIndex, B=8, Comparison Type=::NewEnumerator0)
   ) {module=NMS_CrowdDynamicParameters001}]
  rendererClasses [&NiagaraMeshRendererProperties]
  renderers [
   {rendererClass=&NiagaraMeshRendererProperties, rendererIndex=0}
   {rendererClass=&NiagaraMeshRendererProperties, rendererIndex=1}
  ]
 }}
summary {
 systemName=NS_buffet_hugCrowd
 emitters [{emitterName=crowdTest_v2, rendererClasses [&NiagaraMeshRendererProperties]}]
 userVariables [User.crowdTime {type=NiagaraFloat, defaultValue=0, description=""}]
}
